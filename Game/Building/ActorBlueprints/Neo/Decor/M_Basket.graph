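# M_Basket is a translucent, unlit, two-sided Neo decor material for an animated glow-like effect with two colors driven by face side.
# A time-scrolled mask texture, optionally phase-shifted by actor position, is dotted with vertex color to pick channels and scale an offset along vertex normals, strongest in Z.
# A soft band peaking mid-V is multiplied by inverted Fresnel to fade the UV ends and grazing edges.
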Material /Game/Building/ActorBlueprints/Neo/Decor/M_Basket
graph {
 properties {blendMode=BLEND_Translucent, pixelDepthOffsetMode=PDOM_Legacy, shadingModel=MSM_Unlit, twoSided=true, usedShadingModels=Unlit}
 nodes [
  LinearInterpolate(
   A=VectorParameter() {parameterName=C2, defaultValue=<rgba 1 0 0.707 1>}.RGB
   B=VectorParameter() {parameterName=C1, defaultValue=<rgba 0 0.5445 1 1>}.RGB
   Alpha=TwoSidedSign()
  )
  n1 = Multiply(A=Time(), B=ScalarParameter() {defaultValue=1, group=Adv Emissive, parameterName=AnimMaskTime})
  n2 = ActorPositionWS()
  n3 = Add(
   A=@n1
   B=Divide(
    A=MaterialFunctionCall!(f2, f3=@n2, f4) {materialFunction=/Game/Packages/Fortnite_UniqueMaterials_1/VectorComponentOps}.f3 Add
    B=Constant() {r=50}
   )
  )
  n4 = TextureSampleParameter2D(UVs=Frac(None=StaticSwitchParameter(True=@n3, False=@n1) {group=Adv Emissive, parameterName=Random Timing})) {
   group=Adv Emissive
   parameterName=AnimMask
   samplerSource=SSM_Wrap_WorldGroupSettings
   texture=/Game/Environments/Sets/Neo/Props/Textures/T_AnimMask2
  }
  Multiply(A=Multiply(A=VertexNormalWS(), B=Constant3Vector() {constant=<rgba 4 4 25 1>}), B=DotProduct(A=@n4.RGB, B=VertexColor()))
  n5 = ComponentMask(None=TextureCoordinate()) {g=true}
  n6 = OneMinus(None=@n5)
  Multiply(A=Multiply(A=Multiply(A=@n5, B=@n6), B=Constant() {r=4}), B=OneMinus(None=Fresnel() {baseReflectFraction=0, exponent=2}))
 ]
}
