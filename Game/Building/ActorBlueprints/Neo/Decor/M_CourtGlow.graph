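# Additive unlit glow material, apparently for court decor in the Neo building set.
# Vertex green blends two HDR color parameters, Color2 and Color, to tint the glow.
# Intensity adds a center-peaked U-gradient, scaled by Param_1 and inverted vertex red, to a panning grayscale Time texture.
# That texture samples scaled V of the second UV set on both axes, so it varies along one direction only and ignores the red mask.
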Material /Game/Building/ActorBlueprints/Neo/Decor/M_CourtGlow
graph {
 properties {blendMode=BLEND_Additive, pixelDepthOffsetMode=PDOM_Legacy, shadingModel=MSM_Unlit, usedShadingModels=Unlit}
 nodes [
  n1 = ComponentMask(None=TextureCoordinate()) {r=true}
  n2 = Multiply(A=Multiply(A=@n1, B=OneMinus(None=@n1)), B=ScalarParameter() {defaultValue=2.6389, parameterName=Param_1})
  n3 = OneMinus(None=VertexColor().R)
  n4 = VectorParameter() {defaultValue=<rgba 0.1 0.05 0 0>}
  n5 = MaterialFunctionCall(Float2=Multiply(A=TextureCoordinate() {coordinateIndex=1, uTiling=0.35}, B=ComponentMask(None=@n4.RGB) {g=true, r=true})) {materialFunction=&BreakOutFloat2Components}
  n6 = Panner(Coordinate=AppendVector(A=@n5.G, B=@n5.G), Speed=VectorParameter() {parameterName=Speed, defaultValue=<rgba -0.25 0 0 0>}.RGB)
  Multiply(
   A=LinearInterpolate(
    A=VectorParameter() {parameterName=Color2, defaultValue=<rgba 0 1.6569 8 0>}.RGB
    B=VectorParameter() {parameterName=Color, defaultValue=<rgba 3.9708 0 5 0>}.RGB
    Alpha=VertexColor().G
   )
   B=Add(
    A=Multiply(A=@n2, B=@n3)
    B=TextureSample(UVs=@n6) {samplerType=SAMPLERTYPE_LinearGrayscale, texture=/Game/Environments/Sets/Neo/Cable/Textures/Time}.R
   )
  )
 ]
}
